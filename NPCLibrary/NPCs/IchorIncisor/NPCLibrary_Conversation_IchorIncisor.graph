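# Conversation database for the Ichor Incisor NPC, defining its greeting and the branches a player can reach from it.
# A single entry point leads to a speech node with a default line plus contextual lines for first-ever conversations and specific player outfits.
# Contextual lines take curve-driven priorities.
# The speech node fans out through link nodes to shared NPC quests, basic services and a Season 26 Snapshot quest conversation, referenced by tag.
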ConversationDatabase /NPCLibrary/NPCs/IchorIncisor/NPCLibrary_Conversation_IchorIncisor
entry_points [{tag=Conversation.Entry.IchorIncisor, destinations [#1]}]
nodes [
 ConversationEntryPointNode_7 {
  id=#1
  class=ConversationEntryPointNode
  outputs [#2]
  properties {entryTag {tagName=Conversation.Entry.IchorIncisor}}
 }
 Speech_C_4 {
  id=#2
  class=Speech_C
  outputs [#3, #4, #5]
  properties {
   nodeName=""
   generalConfig {
    defaultMessage="If you spot any paparazzi, let me know. I'm absolutely starving."
    contextualMessages [
     {
      message=Are you looking for an autograph?
      contextRequirements [{requirement {class=FortControllerRequirement_IsFirstEverConversationWithNPC}}]
      priority {curve {rowName=Tandem.Speech.Priority.FirstEverInteraction}}
     }
     {
      message=You think you can slay every monster? That's cute.
      contextRequirements [{requirement {
         class=FortControllerRequirement_HasCID
         properties {allowedCIds [/BRCosmetics/Athena/Items/Cosmetics/Characters/Character_Silencer, /BRCosmetics/Athena/Items/Cosmetics/Characters/Character_ProdigyFire]}
        }}]
      priority {curve {rowName=Tandem.Speech.Priority.OutfitReaction}}
     }
     {
      message=Your friends seem ready to face the darkness... but are you?
      contextRequirements [{requirement {
         class=FortControllerRequirement_HasCID
         properties {allowedCIds [&Character_SharpFang, /BRCosmetics/Athena/Items/Cosmetics/Characters/Character_ProdigySage]}
        }}]
      priority {curve {rowName=Tandem.Speech.Priority.OutfitReaction}}
     }
     {
      message=Do you really think you can protect your friends? Can you even protect yourself?
      contextRequirements [{requirement {
         class=FortControllerRequirement_HasCID
         properties {allowedCIds [/BRCosmetics/Athena/Items/Cosmetics/Characters/Character_Sparrow, /BRCosmetics/Athena/Items/Cosmetics/Characters/Character_ProdigyHaughty]}
        }}]
      priority {curve {rowName=Tandem.Speech.Priority.OutfitReaction}}
     }
     {
      message="Leave me be, Thorne. I have my own business to attend to."
      contextRequirements [{requirement {class=FortControllerRequirement_HasCID, properties {allowedCIds [/BRCosmetics/Athena/Items/Cosmetics/Characters/Character_LazarusLens]}}}]
      priority {curve {rowName=Tandem.Speech.Priority.OutfitReaction}}
     }
     {
      message="I heard you're trying to mentor that little High Stakes Club. Hmph, good luck."
      contextRequirements [{requirement {
         class=FortControllerRequirement_HasCID
         properties {allowedCIds [/BRCosmetics/Athena/Items/Cosmetics/Characters/CID_869_Athena_Commando_F_Cavalry]}
        }}]
      priority {curve {rowName=Tandem.Speech.Priority.OutfitReaction}}
     }
    ]
   }
  }
 }
 ConversationLinkNode_2 {
  id=#5
  class=ConversationLinkNode
  properties {remoteEntryTag {tagName=Conversation.Entry.Season26.Snapshot.P01.Q012.b}}
 }
 ConversationLinkNode_0 {id=#4, class=ConversationLinkNode, properties {remoteEntryTag {tagName=Conversation.Entry.BasicServices}}}
 ConversationLinkNode_1 {id=#3, class=ConversationLinkNode, properties {remoteEntryTag {tagName=Conversation.Entry.NPCLibrary.Quests}}}
]
